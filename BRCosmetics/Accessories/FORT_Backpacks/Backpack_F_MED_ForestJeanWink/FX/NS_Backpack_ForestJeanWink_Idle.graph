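# Idle Niagara effect for the ForestJeanWink backpack cosmetic, emitting twinkling star sprites from authored point arrays offset from a socket.
# One packed emitter uses three spawn groups, a visibility tag and separate sprite attribute sets to feed four sprite renderers.
# System audio reactivity, with a fake-loop module present, drives spawn rate, while amplitude intensity modulates size, velocity, hue and brightness.
# Time-of-day values tint and scale color, and camera offset changes for front end.
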NiagaraSystem /BRCosmetics/Accessories/FORT_Backpacks/Backpack_F_MED_ForestJeanWink/FX/NS_Backpack_ForestJeanWink_Idle
dependencies {
 usedDataInterfaces [
  &NiagaraDataInterfaceArrayFloat3
  &NiagaraDataInterfaceCurve
  &NiagaraDataInterfaceVector2DCurve
  &NiagaraDataInterfaceSocketReader
  &NiagaraDataInterfaceColorCurve
 ]
 usedDynamicInputs [
  &RandomRangeLinearColor
  &SetBoolByFloatComparison
  &ModuloFloat
  &MakeCustomFloatFromBool
  /Niagara/DynamicInputs/Lerp/SmoothLerpOverTime_Float
  &RemapRange
  &MakeFloatFromLinearColor
  &ColorFromCurve
  &FNTimeOfDayScalar
 ]
 usedModules [
  &self:SystemSpawnScript.SystemScriptSource.SystemScriptGraph.NiagaraNodeAssignment_9.SetVariables_#1
  &self:SystemSpawnScript.SystemScriptSource.SystemScriptGraph.NiagaraNodeAssignment_4.SetVariables_#2
  &self:SystemSpawnScript.SystemScriptSource.SystemScriptGraph.NiagaraNodeAssignment_5.SetVariables_#3
  &SystemState
  &self:SystemSpawnScript.SystemScriptSource.SystemScriptGraph.NiagaraNodeAssignment_7.SetVariables_#4
  &Timeline
  /Game/Effects/Niagara/Modules/Audio/FN_FakeAudioReactivityLoop
  &self:SystemSpawnScript.SystemScriptSource.SystemScriptGraph.NiagaraNodeAssignment_1.SetVariables_#5
  &self:SystemSpawnScript.SystemScriptSource.SystemScriptGraph.NiagaraNodeAssignment_0.SetVariables_#6
  &self:SystemSpawnScript.SystemScriptSource.SystemScriptGraph.NiagaraNodeAssignment_3.SetVariables_#7
  &self:SystemSpawnScript.SystemScriptSource.SystemScriptGraph.NiagaraNodeAssignment_2.SetVariables_#8
 ]
}
emitters {Stars_Packed {
  emitterUpdateScript [
   &EmitterState(Inactive Response=::NewEnumerator1, Life Cycle Mode=::NewEnumerator1, Loop Duration=1, UseLoopDelay=false)
   &SpawnRate(SpawnRate=&Lerp_Float(A=0, B=20, Alpha=$System.AudioReactivity), Use Spawn Probability=false)
   &SpawnRate(
    Spawn Probability=0.9
    SpawnGroup=1
    SpawnRate=&Lerp_Float(A=0, B=3, Alpha=$System.AudioReactivity)
    Use Spawn Probability=true
   ) {module=SpawnRate001}
   &SpawnRate(
    Spawn Probability=0.8
    SpawnGroup=2
    SpawnRate=&Lerp_Float(A=0, B=2.5, Alpha=$System.AudioReactivity)
    Use Spawn Probability=true
   ) {module=SpawnRate002}
  ]
  particleSpawnScript [
   &InitializeParticle(
    Color Mode=::NewEnumerator0
    Lifetime=&MergeHelper_FloatSwitch(
     0=&RandomRangeFloat(Minimum=0.35, Maximum=0.5)
     1=&RandomRangeFloat(Minimum=0.18, Maximum=0.225)
     2=&RandomRangeFloat(Minimum=0.25, Maximum=0.325)
     3=&RandomRangeFloat(Minimum=0.18, Maximum=0.225)
     Index=$Emitter.SpawnGroup
     Num=::NewEnumerator1
    )
    Sprite Rotation Angle=&MergeHelper_FloatSwitch(
     0=&RandomRangeFloat(Minimum=0, Maximum=1)
     1=0
     2=0
     3=&RandomRangeFloat(Minimum=0, Maximum=0.2)
     Index=$Emitter.SpawnGroup
     Num=::NewEnumerator1
    )
    Sprite Rotation Mode=::NewEnumerator1
    Sprite Size Mode=::NewEnumerator3
    Sprite UV Mode=::NewEnumerator3
    Uniform Sprite Size=&RandomRangeFloat(Minimum=2, Maximum=4.5)
    UsePositionOffset=false
    Write Material Random=true
   )
   &PartitionParticles(Partition Mode:ENiagara_PartitionMode_v2=::NewEnumerator5)
   &CameraOffset(Camera Offset Amount=&FN_FrontEndFloatModifier(
     Gameplay=&MergeHelper_FloatSwitch(0=12, 1=12, 2=12, 3=?, Index=$Emitter.SpawnGroup, Num=::NewEnumerator1)
     Front End=7
    ))
   /SetVariables(Particles.SampleEdge:NiagaraInt32=&MergeHelper_IntSwitch(
     0=1
     1=&MakeIntFromFloat(Float=&Lerp_Float(A=0, B=1, Alpha=&MakeFloatFromInt(Int=$Particles.PartitionParticles.Partition)))
     2=&MakeIntFromFloat(Float=&Lerp_Float(A=0, B=1, Alpha=&MakeFloatFromInt(Int=$Particles.PartitionParticles.Partition)))
     3=?
     4=?
     5=?
     Index=$Emitter.SpawnGroup
     Num=::NewEnumerator1
    )) {isSetParameters=true}
   &SocketLocation(
    SamplePosOffset=true
    Sampled Position Offset=&MergeHelper_VectorSwitch(
     0=&SelectVectorFromArray(
      Array Sampling Mode=::NewEnumerator0
      Direct Array Index=?
      Vector Selection Array=DataInterface() {propertyValues {FloatData [
         <0.7475 -3.3613 -5.8182>
         <1.466 -0.6648 11.1464>
         <6.0817 -1.1595 -3.1903>
         <7.3561 -1.0095 -9.2225>
         <8.3767 -0.5186 -11.1719>
         <3.5729 -0.7977 -8.6673>
         <1.9459 -3.9286 -1.3121>
         <3.804 -2.847 -5.6331>
         <0.5468 -2.9907 -3.6828>
         <13.5026 -0.5439 4.2668>
         <5.1262 -2.3616 -0.4173>
         <6.8535 -2.7884 2.0836>
         <3.2505 -2.59 3.0423>
         <8.4677 -1.1751 4.2498>
         <7.0246 -0.4489 5.4515>
         <4.7348 -0.9288 5.1094>
         <-8.2868 -0.2874 -8.8417>
         <-6.372 -1.3944 -6.6461>
         <-3.433 -2.9464 -1.9876>
         <-0.9941 -0.2857 -8.0873>
         <-1.4406 -3.1901 -3.85>
         <-8.111 -0.2154 -1.569>
         <-4.6055 -2.797 2.5664>
         <-11.9083 -0.6075 4.5437>
         <-2.3699 -0.3052 10.179>
         <-0.1591 -1.1837 11.981>
         <1.4144 -2.958 3.912>
         <0.492 -1.4028 10.6205>
         <0.5674 -2.0966 8.3155>
         <0.1572 -3.352 -2.9206>
         <-3.3529 -2.1222 4.5336>
        ]}}
      Random Seed=0
     )
     1=&SelectVectorFromArray(
      Array Sampling Mode=::NewEnumerator0
      Direct Array Index=?
      Vector Selection Array=DataInterface() {propertyValues {FloatData [
         <7.6114 -1.5662 -8.6598>
         <7.28 -1.0757 -5.1664>
         <8.6926 -0.6133 -11.6959>
         <5.937 -2.1163 -9.638>
         <1.2338 -0.4671 -8.1991>
         <4.6383 -1.1008 -9.5998>
         <13.6479 -0.863 4.2494>
         <10.1292 -0.3624 0.4522>
         <9.2925 -0.8984 0.0654>
         <8.0569 -0.8291 5.3092>
         <13.5127 -0.815 4.5466>
         <11.6231 -1.3494 4.5078>
         <-7.8819 -1.2768 -9.2525>
         <-7.3806 -1.2476 -6.3021>
         <-8.5549 -0.6243 -10.9477>
         <-6.6919 -1.2168 -10.6156>
         <-7.8032 -1.2362 -11.193>
         <-4.3536 -1.1995 -9.3854>
         <-10.5598 -1.9904 2.1515>
         <-9.0153 -0.182 -0.7672>
         <-13.7566 -0.3951 3.9857>
         <-9.0427 -0.4501 5.3867>
         <-10.4571 -0.8977 4.9309>
         <-10.7329 -2.0036 4.2592>
         <-3.6832 -0.2215 7.8376>
         <-1.9535 -1.5474 9.6669>
         <-0.3205 -0.7073 14.093>
         <3.4727 -0.6807 7.6759>
         <0.4083 -0.4865 14.1983>
         <2.8299 -1.0809 8.4751>
        ]}}
      Random Seed=0
     )
     2=?
     3=?
     Index=$Particles.SampleEdge
     Num=::NewEnumerator0
    )
    Socket Reader=$System.Socket Reader
    Socket Sampling Mode=::NewEnumerator3
    VelocitySpeedLimit=false
   )
   &AddVelocity(
    Module.Use Falloff Distance=false
    Origin Offset=&RotateVectorByQuaternion(Vector To Rotate=<0 1 -4>, Quaternion=$Output.SocketLocation.SampledRotation)
    Override Seed=false
    Velocity Mode=::NewEnumerator1
    Velocity Speed=&MergeHelper_FloatSwitch(
     0=&RandomRangeFloat(Minimum=8, Maximum=15)
     1=&Lerp_Float(
      A=0
      B=&RandomRangeFloat(Minimum=8, Maximum=15)
      Alpha=&MakeFloatFromInt(Int=$Particles.PartitionParticles.Partition)
     )
     2=&Lerp_Float(
      A=0
      B=&RandomRangeFloat(Minimum=8, Maximum=15)
      Alpha=&MakeFloatFromInt(Int=$Particles.PartitionParticles.Partition)
     )
     3=?
     Index=$Emitter.SpawnGroup
     Num=::NewEnumerator1
    )
   )
   /SetVariables(Particles.RotationDirection:NiagaraFloat=&UniformAOrBFloat(A=-1, Evaluation Type=::NewEnumerator0, Recalculate Random Each Loop=true, Override Seed=false)) {isSetParameters=true}
   /SetVariables(Particles.VisibilityTag=$Emitter.SpawnGroup) {isSetParameters=true}
  ]
  particleUpdateScript [
   &ParticleState()
   &ScaleSpriteSize(Uniform Curve Sprite Scale=DataInterface() {propertyValues {Curve {keys [{arriveTangent=3.8099, interpMode=RCIM_Cubic, leaveTangent=3.8099, tangentMode=RCTM_User, time=0, value=0}, <at 0.4 1 RCIM_Cubic>]}}})
   &ScaleSpriteSize(
    Scale Sprite Size Mode=::NewEnumerator0
    Uniform Scale Factor=&FloatFromCurve(
     FloatCurve=DataInterface() {propertyValues {Curve {keys [
         {arriveTangent=1.0554, interpMode=RCIM_Cubic, leaveTangent=1.0554, tangentMode=RCTM_User, time=0, value=1}
         {arriveTangentWeight=0.3334, interpMode=RCIM_Cubic, tangentMode=RCTM_User, time=1, value=1.4}
        ]}}}
     CurveIndex=$System.AmplitudeIntensity
    )
   ) {module=ScaleSpriteSize001}
   &SpriteRotationRate(Rotation Rate=&Multiply_Float(
     A=$Particles.RotationDirection
     B=&Multiply_Float(
      A=&RandomRangeFloat(Minimum=20, Maximum=35)
      B=&MergeHelper_FloatSwitch(0=1, 1=0.8, 2=0, 3=?, Index=$Particles.VisibilityTag, Num=::NewEnumerator1)
     )
    ))
   &ScaleVelocity(Velocity Scale=&VectorFromFloat(Value=&Lerp_Float(A=0.55, B=1.2, Alpha=$System.AmplitudeIntensity)))
   &ScaleColor(
    Scale Alpha=&FloatFromCurve(FloatCurve=DataInterface() {propertyValues {Curve {keys [{arriveTangent=-1.4967, interpMode=RCIM_Cubic, time=0.4, value=1}, {interpMode=RCIM_Cubic, leaveTangent=-3.0489, time=1, value=0}]}}})
    Scale RGB=&Multiply_Vector(
     A=&MakeVectorFromLinearColorRGB(Color=&HueShiftLinearColor(
       Hue=&Lerp_Float(
        A=0
        B=&UniformAOrBFloat(A=-0.0025, B=0.0025, Evaluation Type=::NewEnumerator0, Recalculate Random Each Loop=true, Override Seed=false)
        Alpha=$System.AmplitudeIntensity
       )
       Saturation=1
       Value=&FloatFromCurve(
        FloatCurve=DataInterface() {propertyValues {Curve {keys [{arriveTangent=1.9866, interpMode=RCIM_Cubic, leaveTangent=1.9866, tangentMode=RCTM_User, time=0, value=1}, <at 1 2 RCIM_Cubic>]}}}
        CurveIndex=$System.AmplitudeIntensity
       )
       Linear Color=&LerpLinearColors(
        StartColor=<rgba 1 0.1274 0.402 1>
        EndColor=<rgba 1 0.0953 0.08 1>
        LerpFactor=&UniformAOrBFloat(A=0, Evaluation Type=::NewEnumerator0, Recalculate Random Each Loop=true, Override Seed=false)
       )
      ))
     B=&MakeVectorFromLinearColorRGB(Color=$System.TODColor)
    )
    ScaleA=true
    ScaleRGB=true
   )
   &FN_Sprite_Attributes(
    Color=&MakeLinearColorFromVectorAndFloat(
     "Vector (RGB)"=&MakeVectorFromLinearColorRGB(Color=$Particles.Color)
     "Float (Alpha)"=&FloatFromCurve(FloatCurve=DataInterface() {propertyValues {Curve {keys [<at 0.5 1>,<at 1 0>]}}}, Scale Curve=0.065)
    )
    Create Color Attribute=true
    Create SpriteSize Attribute=true
    SpriteSize=&Multiply_Vector2D_ByFloat(Float=2.25, Vector2D=$Particles.SpriteSize)
    Write to Module Namespace=true
   ) {module=FN_Sprite_Attributes002}
   &FN_Sprite_Attributes(
    Color=&MakeLinearColorFromVectorAndFloat(
     "Vector (RGB)"=&Multiply_VectorByFloat(
      Vector=&MakeVectorFromLinearColorRGB(Color=<rgba 3 3 3 1>)
      Float=&Multiply_Float(A=&Lerp_Float(A=1, B=3, Alpha=$System.AmplitudeIntensity), B=$System.TODScalar)
     )
     "Float (Alpha)"=&FloatFromCurve(
      FloatCurve=DataInterface() {propertyValues {Curve {keys [
          {arriveTangent=-8.3354, interpMode=RCIM_Cubic, leaveTangent=-8.3354, tangentMode=RCTM_User, time=0.1, value=3}
          {interpMode=RCIM_Cubic, leaveTangent=-2.9611, time=1, value=0}
         ]}}}
      Scale Curve=&Multiply_Float(
       A=1
       B=&Lerp_Float(
        A=1
        B=&UniformAOrBFloat(A=0, Evaluation Type=::NewEnumerator1, Recalculate Random Each Loop=true, Override Seed=false)
        Alpha=&FloatFromCurve(FloatCurve=DataInterface() {propertyValues {Curve {keys [<at 0.225 1>,<at 0.4 0>]}}})
       )
      )
     )
    )
    Create Color Attribute=true
    Create SpriteSize Attribute=true
    SpriteSize=&Multiply_Vector2D_ByFloat(
     Float=&Lerp_Float(A=1, B=1.2, Alpha=$System.AmplitudeIntensity)
     Vector2D=&Vector2DFromFloat(Value=&RandomRangeFloat(Minimum=2.25, Maximum=4.25))
    )
    Write to Module Namespace=true
   ) {module=FN_Sprite_Attributes001}
   &FN_Sprite_Attributes(
    Color=&MakeLinearColorFromVectorAndFloat(
     "Vector (RGB)"=&Multiply_VectorByFloat(
      Vector=&MakeVectorFromLinearColorRGB(Color=<rgba 3 3 3 1>)
      Float=&Multiply_Float(A=&Lerp_Float(A=1, B=5, Alpha=$System.AmplitudeIntensity), B=$System.TODScalar)
     )
     "Float (Alpha)"=&FloatFromCurve(
      FloatCurve=DataInterface() {propertyValues {Curve {keys [{arriveTangent=-1.5731, interpMode=RCIM_Cubic, time=0.4, value=1}, {interpMode=RCIM_Cubic, leaveTangent=-2.9611, time=1, value=0}]}}}
      Scale Curve=&Lerp_Float(
       A=1
       B=&UniformAOrBFloat(A=0, Evaluation Type=::NewEnumerator1, Recalculate Random Each Loop=true, Override Seed=false)
       Alpha=&FloatFromCurve(FloatCurve=DataInterface() {propertyValues {Curve {keys [<at 0.4 1>,<at 1 0>]}}})
      )
     )
    )
    Create Color Attribute=true
    Create SpriteSize Attribute=true
    SpriteSize=&Multiply_Vector2D_ByFloat(
     Float=&Lerp_Float(A=1, B=1.2, Alpha=$System.AmplitudeIntensity)
     Vector2D=&Multiply_Vector2D(
      A=&Vector2DFromFloat(Value=&RandomRangeFloat(Minimum=4, Maximum=7.5))
      B=&Vector2DFromCurve(
       Vector2Curve=DataInterface() {propertyValues {
         XCurve {keys [{arriveTangent=-0.7214, interpMode=RCIM_Cubic, leaveTangent=-0.7214, tangentMode=RCTM_User, time=0.25, value=1}, <at 1 0.3>]}
         YCurve {keys [
           <at 0 1 RCIM_Cubic>
           {arriveTangent=-2.1615, interpMode=RCIM_Cubic, leaveTangent=-2.1615, tangentMode=RCTM_User, time=0.15, value=1.4}
           <at 1 0 RCIM_Cubic>
          ]}
        }}
       ScaleCurve=<1 1>
      )
     )
    )
    Write to Module Namespace=true
   )
   /SetVariables(Particles.DynamicParametersAlt:Vector4f=&MakeVector4(
     X=&FloatFromCurve(FloatCurve=DataInterface() {propertyValues {Curve {keys [<at 0 2 RCIM_Cubic>, {arriveTangent=7.0068, interpMode=RCIM_Cubic, leaveTangent=7.0068, tangentMode=RCTM_User, time=1, value=6}]}}})
     Y=0
     Z=0
     W=0
    )) {isSetParameters=true}
   &SolveForcesAndVelocity(Clamp Velocity=false, Limit Acceleration=false)
  ]
  rendererClasses [&NiagaraSpriteRendererProperties]
  renderers [
   {rendererClass=&NiagaraSpriteRendererProperties}
   {rendererClass=&NiagaraSpriteRendererProperties, rendererIndex=1}
   {rendererClass=&NiagaraSpriteRendererProperties, rendererIndex=2}
   {rendererClass=&NiagaraSpriteRendererProperties, rendererIndex=3}
  ]
 }}
summary {
 systemName=NS_Backpack_ForestJeanWink_Idle
 emitters [{emitterName=Stars_Packed, rendererClasses [&NiagaraSpriteRendererProperties]}]
 userVariables [
  User.bIsFrontEnd {type=NiagaraBool}
  User.bIsFrontEndPreview {type=NiagaraBool}
  User.DebugReactivity {type=NiagaraBool}
 ]
}
