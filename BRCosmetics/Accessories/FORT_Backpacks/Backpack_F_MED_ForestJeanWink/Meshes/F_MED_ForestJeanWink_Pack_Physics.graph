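# Physics asset for the ForestJeanWink backpack cosmetic, setting up physical motion for its tassels in bottom, left, right, middle and top groups.
# Each base tassel bone anchors a chain of one or two dyn bones, all using identical small capsule bodies.
# Angular constraints link each chain, with swing limits generally widest on bottom tassels and tightest on middle and top ones.
# Top tassels also have twist locked, while the others allow slight twist.
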PhysicsAsset /BRCosmetics/Accessories/FORT_Backpacks/Backpack_F_MED_ForestJeanWink/Meshes/F_MED_ForestJeanWink_Pack_Physics
bodies {
 base_tassel_bottom_01 [{shapeType=Capsule, length=0.505, radius=0.505, shapeName=base_tassel_bottom_01_capsule, rotation=<rot 0 -90 0>}]
 base_tassel_bottom_02 [{shapeType=Capsule, length=0.505, radius=0.505, shapeName=base_tassel_bottom_02_capsule, rotation=<rot 0 -90 0>}]
 base_tassel_bottom_03 [{shapeType=Capsule, length=0.505, radius=0.505, shapeName=base_tassel_bottom_03_capsule, rotation=<rot 0 -90 0>}]
 base_tassel_bottom_04 [{shapeType=Capsule, length=0.505, radius=0.505, shapeName=base_tassel_bottom_04_capsule, rotation=<rot 0 -90 0>}]
 base_tassel_bottom_05 [{shapeType=Capsule, length=0.505, radius=0.505, shapeName=base_tassel_bottom_05_capsule, rotation=<rot 0 -90 0>}]
 base_tassel_bottom_06 [{shapeType=Capsule, length=0.505, radius=0.505, shapeName=base_tassel_bottom_06_capsule, rotation=<rot 0 -90 0>}]
 base_tassel_bottom_07 [{shapeType=Capsule, length=0.505, radius=0.505, shapeName=base_tassel_bottom_07_capsule, rotation=<rot 0 -90 0>}]
 base_tassel_bottom_08 [{shapeType=Capsule, length=0.505, radius=0.505, shapeName=base_tassel_bottom_08_capsule, rotation=<rot 0 -90 0>}]
 base_tassel_bottom_09 [{shapeType=Capsule, length=0.505, radius=0.505, shapeName=base_tassel_bottom_09_capsule, rotation=<rot 0 -90 0>}]
 base_tassel_left_01 [{shapeType=Capsule, length=0.505, radius=0.505, shapeName=base_tassel_left_01_capsule, rotation=<rot 0 -90 0>}]
 base_tassel_left_02 [{shapeType=Capsule, length=0.505, radius=0.505, shapeName=base_tassel_left_02_capsule, rotation=<rot 0 -90 0>}]
 base_tassel_left_03 [{shapeType=Capsule, length=0.505, radius=0.505, shapeName=base_tassel_left_03_capsule, rotation=<rot 0 -90 0>}]
 base_tassel_middle_01 [{shapeType=Capsule, length=0.505, radius=0.505, shapeName=base_tassel_middle_01_capsule, rotation=<rot 0 -90 0>}]
 base_tassel_middle_02 [{shapeType=Capsule, length=0.505, radius=0.505, shapeName=base_tassel_middle_02_capsule, rotation=<rot 0 -90 0>}]
 base_tassel_middle_03 [{shapeType=Capsule, length=0.505, radius=0.505, shapeName=base_tassel_middle_03_capsule, rotation=<rot 0 -90 0>}]
 base_tassel_right_01 [{shapeType=Capsule, length=0.505, radius=0.505, shapeName=base_tassel_right_01_capsule, rotation=<rot 0 -90 0>}]
 base_tassel_right_02 [{shapeType=Capsule, length=0.505, radius=0.505, shapeName=base_tassel_right_02_capsule, rotation=<rot 0 -90 0>}]
 base_tassel_right_03 [{shapeType=Capsule, length=0.505, radius=0.505, shapeName=base_tassel_right_03_capsule, rotation=<rot 0 -90 0>}]
 base_tassel_top_01 [{shapeType=Capsule, length=0.505, radius=0.505, shapeName=base_tassel_top_01_capsule, rotation=<rot 0 -90 0>}]
 base_tassel_top_02 [{shapeType=Capsule, length=0.505, radius=0.505, shapeName=base_tassel_top_02_capsule, rotation=<rot 0 -90 0>}]
 base_tassel_top_03 [{shapeType=Capsule, length=0.505, radius=0.505, shapeName=base_tassel_top_03_capsule, rotation=<rot 0 -90 0>}]
 dyn_tassel_bottom_01_01 [{shapeType=Capsule, length=0.505, radius=0.505, shapeName=dyn_tassel_bottom_01_01_capsule, center=<1 0 0>, rotation=<rot 90 0 0>}]
 dyn_tassel_bottom_02_01 [{shapeType=Capsule, length=0.505, radius=0.505, shapeName=dyn_tassel_bottom_02_01_capsule, center=<1 0 0>, rotation=<rot 90 0 0>}]
 dyn_tassel_bottom_02_02 [{shapeType=Capsule, length=0.505, radius=0.505, shapeName=dyn_tassel_bottom_02_02_capsule, center=<1 0 0>, rotation=<rot 90 0 0>}]
 dyn_tassel_bottom_03_01 [{shapeType=Capsule, length=0.505, radius=0.505, shapeName=dyn_tassel_bottom_03_01_capsule, center=<1 0 0>, rotation=<rot 90 0 0>}]
 dyn_tassel_bottom_03_02 [{shapeType=Capsule, length=0.505, radius=0.505, shapeName=dyn_tassel_bottom_03_02_capsule, center=<1 0 0>, rotation=<rot 90 0 0>}]
 dyn_tassel_bottom_04_01 [{shapeType=Capsule, length=0.505, radius=0.505, shapeName=dyn_tassel_bottom_04_01_capsule, center=<1 0 0>, rotation=<rot 90 0 0>}]
 dyn_tassel_bottom_04_02 [{shapeType=Capsule, length=0.505, radius=0.505, shapeName=dyn_tassel_bottom_04_02_capsule, center=<1 0 0>, rotation=<rot 90 0 0>}]
 dyn_tassel_bottom_05_01 [{shapeType=Capsule, length=0.505, radius=0.505, shapeName=dyn_tassel_bottom_05_01_capsule, center=<1 0 0>, rotation=<rot 90 0 0>}]
 dyn_tassel_bottom_05_02 [{shapeType=Capsule, length=0.505, radius=0.505, shapeName=dyn_tassel_bottom_05_02_capsule, center=<1 0 0>, rotation=<rot 90 0 0>}]
 dyn_tassel_bottom_06_01 [{shapeType=Capsule, length=0.505, radius=0.505, shapeName=dyn_tassel_bottom_06_01_capsule, center=<1 0 0>, rotation=<rot 90 0 0>}]
 dyn_tassel_bottom_06_02 [{shapeType=Capsule, length=0.505, radius=0.505, shapeName=dyn_tassel_bottom_06_02_capsule, center=<1 0 0>, rotation=<rot 90 0 0>}]
 dyn_tassel_bottom_07_01 [{shapeType=Capsule, length=0.505, radius=0.505, shapeName=dyn_tassel_bottom_07_01_capsule, center=<1 0 0>, rotation=<rot 90 0 0>}]
 dyn_tassel_bottom_07_02 [{shapeType=Capsule, length=0.505, radius=0.505, shapeName=dyn_tassel_bottom_07_02_capsule, center=<1 0 0>, rotation=<rot 90 0 0>}]
 dyn_tassel_bottom_08_01 [{shapeType=Capsule, length=0.505, radius=0.505, shapeName=dyn_tassel_bottom_08_01_capsule, center=<1 0 0>, rotation=<rot 90 0 0>}]
 dyn_tassel_bottom_08_02 [{shapeType=Capsule, length=0.505, radius=0.505, shapeName=dyn_tassel_bottom_08_02_capsule, center=<1 0 0>, rotation=<rot 90 0 0>}]
 dyn_tassel_bottom_09_01 [{shapeType=Capsule, length=0.505, radius=0.505, shapeName=dyn_tassel_bottom_09_01_capsule, center=<1 0 0>, rotation=<rot 90 0 0>}]
 dyn_tassel_bottom_09_02 [{shapeType=Capsule, length=0.505, radius=0.505, shapeName=dyn_tassel_bottom_09_02_capsule, center=<1 0 0>, rotation=<rot 90 0 0>}]
 dyn_tassel_left_01_01 [{shapeType=Capsule, length=0.505, radius=0.505, shapeName=dyn_tassel_left_01_01_capsule, center=<1 0 0>, rotation=<rot 90 0 0>}]
 dyn_tassel_left_01_02 [{shapeType=Capsule, length=0.505, radius=0.505, shapeName=dyn_tassel_left_01_02_capsule, center=<1 0 0>, rotation=<rot 90 0 0>}]
 dyn_tassel_left_02_01 [{shapeType=Capsule, length=0.505, radius=0.505, shapeName=dyn_tassel_left_02_01_capsule, center=<1 0 0>, rotation=<rot 90 0 0>}]
 dyn_tassel_left_02_02 [{shapeType=Capsule, length=0.505, radius=0.505, shapeName=dyn_tassel_left_02_02_capsule, center=<1 0 0>, rotation=<rot 90 0 0>}]
 dyn_tassel_left_03_01 [{shapeType=Capsule, length=0.505, radius=0.505, shapeName=dyn_tassel_left_03_01_capsule, center=<1 0 0>, rotation=<rot 90 0 0>}]
 dyn_tassel_left_03_02 [{shapeType=Capsule, length=0.505, radius=0.505, shapeName=dyn_tassel_left_03_02_capsule, center=<1 0 0>, rotation=<rot 90 0 0>}]
 dyn_tassel_middle_01_01 [{shapeType=Capsule, length=0.505, radius=0.505, shapeName=dyn_tassel_middle_01_01_capsule, center=<1 0 0>, rotation=<rot 90 0 0>}]
 dyn_tassel_middle_01_02 [{shapeType=Capsule, length=0.505, radius=0.505, shapeName=dyn_tassel_middle_01_02_capsule, center=<1 0 0>, rotation=<rot 90 0 0>}]
 dyn_tassel_middle_02_01 [{shapeType=Capsule, length=0.505, radius=0.505, shapeName=dyn_tassel_middle_02_01_capsule, center=<1 0 0>, rotation=<rot 90 0 0>}]
 dyn_tassel_middle_03_01 [{shapeType=Capsule, length=0.505, radius=0.505, shapeName=dyn_tassel_middle_03_01_capsule, center=<1 0 0>, rotation=<rot 90 0 0>}]
 dyn_tassel_right_01_01 [{shapeType=Capsule, length=0.505, radius=0.505, shapeName=dyn_tassel_right_01_01_capsule, center=<1 0 0>, rotation=<rot 90 0 0>}]
 dyn_tassel_right_01_02 [{shapeType=Capsule, length=0.505, radius=0.505, shapeName=dyn_tassel_right_01_02_capsule, center=<1 0 0>, rotation=<rot 90 0 0>}]
 dyn_tassel_right_02_01 [{shapeType=Capsule, length=0.505, radius=0.505, shapeName=dyn_tassel_right_02_01_capsule, center=<1 0 0>, rotation=<rot 90 0 0>}]
 dyn_tassel_right_02_02 [{shapeType=Capsule, length=0.505, radius=0.505, shapeName=dyn_tassel_right_02_02_capsule, center=<1 0 0>, rotation=<rot 90 0 0>}]
 dyn_tassel_right_03_01 [{shapeType=Capsule, length=0.505, radius=0.505, shapeName=dyn_tassel_right_03_01_capsule, center=<1 0 0>, rotation=<rot 90 0 0>}]
 dyn_tassel_right_03_02 [{shapeType=Capsule, length=0.505, radius=0.505, shapeName=dyn_tassel_right_03_02_capsule, center=<1 0 0>, rotation=<rot 90 0 0>}]
 dyn_tassel_top_01_01 [{shapeType=Capsule, length=0.505, radius=0.505, shapeName=dyn_tassel_top_01_01_capsule, center=<1 0 0>, rotation=<rot 90 0 0>}]
 dyn_tassel_top_01_02 [{shapeType=Capsule, length=0.505, radius=0.505, shapeName=dyn_tassel_top_01_02_capsule, center=<1 0 0>, rotation=<rot 90 0 0>}]
 dyn_tassel_top_02_01 [{shapeType=Capsule, length=0.505, radius=0.505, shapeName=dyn_tassel_top_02_01_capsule, center=<1 0 0>, rotation=<rot 90 0 0>}]
 dyn_tassel_top_02_02 [{shapeType=Capsule, length=0.505, radius=0.505, shapeName=dyn_tassel_top_02_02_capsule, center=<1 0 0>, rotation=<rot 90 0 0>}]
 dyn_tassel_top_03_01 [{shapeType=Capsule, length=0.505, radius=0.505, shapeName=dyn_tassel_top_03_01_capsule, center=<1 0 0>, rotation=<rot 90 0 0>}]
 dyn_tassel_top_03_02 [{shapeType=Capsule, length=0.505, radius=0.505, shapeName=dyn_tassel_top_03_02_capsule, center=<1 0 0>, rotation=<rot 90 0 0>}]
}
constraints [
 {bone1Name=dyn_tassel_bottom_01_01, bone2Name=base_tassel_bottom_01, swing1LimitDegrees=60, swing2LimitDegrees=20, twistLimitDegrees=5}
 {bone1Name=dyn_tassel_bottom_02_01, bone2Name=base_tassel_bottom_02, swing1LimitDegrees=60, swing2LimitDegrees=20, twistLimitDegrees=5}
 {bone1Name=dyn_tassel_bottom_02_02, bone2Name=dyn_tassel_bottom_02_01, swing1LimitDegrees=60, swing2LimitDegrees=20, twistLimitDegrees=5}
 {bone1Name=dyn_tassel_bottom_03_01, bone2Name=base_tassel_bottom_03, swing1LimitDegrees=50, swing2LimitDegrees=20, twistLimitDegrees=5}
 {bone1Name=dyn_tassel_bottom_03_02, bone2Name=dyn_tassel_bottom_03_01, swing1LimitDegrees=50, swing2LimitDegrees=20, twistLimitDegrees=5}
 {bone1Name=dyn_tassel_bottom_04_01, bone2Name=base_tassel_bottom_04, swing1LimitDegrees=40, swing2LimitDegrees=15, twistLimitDegrees=5}
 {bone1Name=dyn_tassel_bottom_04_02, bone2Name=dyn_tassel_bottom_04_01, swing1LimitDegrees=40, swing2LimitDegrees=15, twistLimitDegrees=5}
 {bone1Name=dyn_tassel_bottom_05_01, bone2Name=base_tassel_bottom_05, swing1LimitDegrees=60, swing2LimitDegrees=20, twistLimitDegrees=5}
 {bone1Name=dyn_tassel_bottom_05_02, bone2Name=dyn_tassel_bottom_05_01, swing1LimitDegrees=60, swing2LimitDegrees=20, twistLimitDegrees=5}
 {bone1Name=dyn_tassel_bottom_06_01, bone2Name=base_tassel_bottom_06, swing1LimitDegrees=50, swing2LimitDegrees=20, twistLimitDegrees=5}
 {bone1Name=dyn_tassel_bottom_06_02, bone2Name=dyn_tassel_bottom_06_01, swing1LimitDegrees=50, swing2LimitDegrees=20, twistLimitDegrees=5}
 {bone1Name=dyn_tassel_bottom_07_01, bone2Name=base_tassel_bottom_07, swing1LimitDegrees=40, swing2LimitDegrees=15, twistLimitDegrees=5}
 {bone1Name=dyn_tassel_bottom_07_02, bone2Name=dyn_tassel_bottom_07_01, swing1LimitDegrees=40, swing2LimitDegrees=15, twistLimitDegrees=5}
 {bone1Name=dyn_tassel_bottom_08_01, bone2Name=base_tassel_bottom_08, swing1LimitDegrees=50, swing2LimitDegrees=20, twistLimitDegrees=5}
 {bone1Name=dyn_tassel_bottom_08_02, bone2Name=dyn_tassel_bottom_08_01, swing1LimitDegrees=50, swing2LimitDegrees=20, twistLimitDegrees=5}
 {bone1Name=dyn_tassel_bottom_09_01, bone2Name=base_tassel_bottom_09, swing1LimitDegrees=60, swing2LimitDegrees=20, twistLimitDegrees=5}
 {bone1Name=dyn_tassel_bottom_09_02, bone2Name=dyn_tassel_bottom_09_01, swing1LimitDegrees=60, swing2LimitDegrees=20, twistLimitDegrees=5}
 {bone1Name=dyn_tassel_left_01_01, bone2Name=base_tassel_left_01, swing1LimitDegrees=40, swing2LimitDegrees=20, twistLimitDegrees=5}
 {bone1Name=dyn_tassel_left_01_02, bone2Name=dyn_tassel_left_01_01, swing1LimitDegrees=40, swing2LimitDegrees=20, twistLimitDegrees=5}
 {bone1Name=dyn_tassel_left_02_01, bone2Name=base_tassel_left_02, swing1LimitDegrees=40, swing2LimitDegrees=20, twistLimitDegrees=5}
 {bone1Name=dyn_tassel_left_02_02, bone2Name=dyn_tassel_left_02_01, swing1LimitDegrees=40, swing2LimitDegrees=20, twistLimitDegrees=5}
 {bone1Name=dyn_tassel_left_03_01, bone2Name=base_tassel_left_03, swing1LimitDegrees=40, swing2LimitDegrees=20, twistLimitDegrees=5}
 {bone1Name=dyn_tassel_left_03_02, bone2Name=dyn_tassel_left_03_01, swing1LimitDegrees=40, swing2LimitDegrees=20, twistLimitDegrees=5}
 {bone1Name=dyn_tassel_middle_01_01, bone2Name=base_tassel_middle_01, swing1LimitDegrees=20, swing2LimitDegrees=10, twistLimitDegrees=5}
 {bone1Name=dyn_tassel_middle_01_02, bone2Name=dyn_tassel_middle_01_01, swing1LimitDegrees=20, swing2LimitDegrees=10, twistLimitDegrees=5}
 {bone1Name=dyn_tassel_middle_02_01, bone2Name=base_tassel_middle_02, swing1LimitDegrees=20, swing2LimitDegrees=10, twistLimitDegrees=5}
 {bone1Name=dyn_tassel_middle_03_01, bone2Name=base_tassel_middle_03, swing1LimitDegrees=20, swing2LimitDegrees=10, twistLimitDegrees=5}
 {bone1Name=dyn_tassel_right_01_01, bone2Name=base_tassel_right_01, swing1LimitDegrees=40, swing2LimitDegrees=20, twistLimitDegrees=5}
 {bone1Name=dyn_tassel_right_01_02, bone2Name=dyn_tassel_right_01_01, swing1LimitDegrees=40, swing2LimitDegrees=20, twistLimitDegrees=5}
 {bone1Name=dyn_tassel_right_02_01, bone2Name=base_tassel_right_02, swing1LimitDegrees=40, swing2LimitDegrees=20, twistLimitDegrees=5}
 {bone1Name=dyn_tassel_right_02_02, bone2Name=dyn_tassel_right_02_01, swing1LimitDegrees=40, swing2LimitDegrees=20, twistLimitDegrees=5}
 {bone1Name=dyn_tassel_right_03_01, bone2Name=base_tassel_right_03, swing1LimitDegrees=40, swing2LimitDegrees=20, twistLimitDegrees=5}
 {bone1Name=dyn_tassel_right_03_02, bone2Name=dyn_tassel_right_03_01, swing1LimitDegrees=40, swing2LimitDegrees=20, twistLimitDegrees=5}
 {
  bone1Name=dyn_tassel_top_01_01
  bone2Name=base_tassel_top_01
  swing1LimitDegrees=20
  swing2LimitDegrees=0
  twistLimitDegrees=0
  twistMotion=Locked
 }
 {
  bone1Name=dyn_tassel_top_01_02
  bone2Name=dyn_tassel_top_01_01
  swing1LimitDegrees=20
  swing2LimitDegrees=0
  twistLimitDegrees=0
  twistMotion=Locked
 }
 {
  bone1Name=dyn_tassel_top_02_01
  bone2Name=base_tassel_top_02
  swing1LimitDegrees=20
  swing2LimitDegrees=10
  twistLimitDegrees=0
  twistMotion=Locked
 }
 {
  bone1Name=dyn_tassel_top_02_02
  bone2Name=dyn_tassel_top_02_01
  swing1LimitDegrees=20
  swing2LimitDegrees=10
  twistLimitDegrees=0
  twistMotion=Locked
 }
 {
  bone1Name=dyn_tassel_top_03_01
  bone2Name=base_tassel_top_03
  swing1LimitDegrees=20
  swing2LimitDegrees=10
  twistLimitDegrees=0
  twistMotion=Locked
 }
 {
  bone1Name=dyn_tassel_top_03_02
  bone2Name=dyn_tassel_top_03_01
  swing1LimitDegrees=20
  swing2LimitDegrees=10
  twistLimitDegrees=0
  twistMotion=Locked
 }
]
