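# Animation Blueprint for the ForestJeanWink backpack cosmetic that spins backpack_master to music when the Parts variant enables reactivity.
# EventGraph smooths the music amplitude from a material parameter collection and toggles MusicActive through thresholds, driving bone-transform yaw and alpha.
# In the front end without an emote, a looping timer triggers timed spins instead.
# A cached pose feeds two RigidBody branches, selected by bIsInFrontEnd, with the front-end branch adding its own rotation.
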AnimBlueprint /BRCosmetics/Accessories/FORT_Backpacks/Backpack_F_MED_ForestJeanWink/Meshes/F_MED_ForestJeanWink_Pack_AnimBP
target_skeleton=&/F_MED_ForestJeanWink_Pack_Skeleton
graphs [
 AnimGraph {nodes [
   n1 = "Transform (Modify) Bone - Bone: backpack_master"(
    ComponentPose=Local To Component(LocalPose=Use cached pose 'Pre Dynamics'())
    Alpha=1
    Translation="0,-9,0"
    Rotation=Get FE Rotation() -> (FE Rotation:Rotator)
    Scale="1,1,1"
   ) {
    rotationMode=BMM_Additive
    rotationSpace=BCS_BoneSpace
    scaleMode=BMM_Additive
    scaleSpace=BCS_BoneSpace
    translationMode=BMM_Additive
    translationSpace=BCS_BoneSpace
    alphaScaleBiasClamp {bInterpResult=true, interpSpeedDecreasing=5, interpSpeedIncreasing=4, inRange {max=0.8}}
    boneToModify {boneName=backpack_master}
    translation {y=-9}
   }
   n2 = RigidBody!(ComponentPose=@n1, Alpha=0.3, OverrideWorldGravity:Vector="0,0,0", ExternalForce) -> (Pose) {
    bClampLinearTranslationLimitToRefPose=true
    bOverrideWorldGravity=true
    overridePhysicsAsset=&/F_MED_ForestJeanWink_Pack_Physics
    simulationSpace=WorldSpace
    baseBoneRef {boneName=pack}
    componentAppliedLinearAccClamp=<0 0 0>
    simSpaceSettings {
     dampingAlpha=0
     maxAngularAcceleration=16
     maxAngularVelocity=8
     maxLinearAcceleration=2100
     maxLinearVelocity=700
     worldAlpha=0.1
     externalLinearDragV=<0.5 0.5 0.5>
    }
   }
   n3 = RigidBody(ComponentPose=Local To Component(LocalPose=Use cached pose 'Pre Dynamics'()), OverrideWorldGravity="0,0,0") {
    bClampLinearTranslationLimitToRefPose=true
    bOverrideWorldGravity=true
    overridePhysicsAsset=&F_MED_ForestJeanWink_Pack_Physics
    simulationSpace=BaseBoneSpace
    baseBoneRef {boneName=backpack_master}
    componentAppliedLinearAccClamp=<0 0 0>
    simSpaceSettings {
     dampingAlpha=0
     maxAngularAcceleration=16
     maxAngularVelocity=8
     maxLinearAcceleration=2100
     maxLinearVelocity=700
     worldAlpha=1
     externalLinearDragV=<0.5 0.5 0.5>
    }
   }
   Output Pose(Result=Blend Poses by bool(BlendPose_0=Component To Local(ComponentPose=@n2), BlendPose_1=Component To Local(ComponentPose=@n3), bActiveValue=Get bIsInFrontEnd())) {layerGroup=None}
   n4 = "Transform (Modify) Bone - Bone: pack"(ComponentPose=Local To Component(LocalPose=Input Pose()), Translation="0,0.7,0", Rotation="0,0,0") {
    rotationMode=BMM_Additive
    rotationSpace=BCS_BoneSpace
    translationMode=BMM_Additive
    translationSpace=BCS_BoneSpace
    boneToModify {boneName=pack}
    translation {y=0.7}
   }
   n5 = "Transform (Modify) Bone - Bone: backpack_master"(
    ComponentPose=@n4
    Alpha=Get ReactiveToMusic() -> (ReactiveToMusic:float)
    Translation="0,-9,0"
    Rotation=Get Rotation Speed() -> (Rotation Speed:Rotator)
    Scale="1,1,1"
   ) {
    rotationMode=BMM_Additive
    rotationSpace=BCS_BoneSpace
    scaleMode=BMM_Additive
    scaleSpace=BCS_BoneSpace
    translationMode=BMM_Additive
    translationSpace=BCS_BoneSpace
    alphaScaleBiasClamp {bInterpResult=true, interpSpeedDecreasing=5, interpSpeedIncreasing=15, inRange {max=0.8}}
    boneToModify {boneName=backpack_master}
    translation {y=-9}
   }
   Save cached pose 'Pre Dynamics'(Pose=Component To Local(ComponentPose=@n5)) {cacheName=Pre Dynamics}
   Get Rotation Speed()
  ]}
 EventGraph {nodes [
   n1 = GetScalarParameterValue!(
    execute
    Collection:Material Parameter Collection Object Reference=Get MPC_Audio_Reactive_Parameters() -> (MPC_Audio_Reactive_Parameters:Material Parameter Collection Object Reference)
    ParameterName:Name=Get Amplitude() -> (Amplitude:Name=None)
   ) -> (then, ReturnValue:float) {is=title K2Node_CallMaterialParameterCollectionFunction}
   > Set Amplitude_Value(Amplitude_Value:float=FInterpTo(Current=Get Amplitude_Value() -> (Amplitude_Value:float), Target=@n1, DeltaTime=Get Delta Time X() -> (Delta Time X:double), InterpSpeed=8)) -> (Output_Get:float)
   n2 = GetScalarParameterValue(Collection=Get MPC_Audio_Reactive_Parameters(), ParameterName=Get Amplitude() -> (Amplitude=None))
   > Set ReactiveToMusic(ReactiveToMusic:float=SelectFloat(
     A=Remap Float(Input=@n2, OldRange=(X=0,Y=0.2), NewRange=(X=0,Y=1), Clamp=true) -> (Output=~)
     B=0
     bPickA=Get Reactivity On/Off() -> (Reactivity On/Off:bool)
    )) -> (Output_Get:float)
   n3 = Reactivity() {is=title K2Node_CustomEvent}
   > Branch(Condition=Get Reactivity On/Off())
   > Branch(Condition=Get bIsInFrontEnd())
   > n4 = Branch(Condition=Get bIsPlayingEmote())
   Set Time(execute=@n4.else, Time:double=0) -> (Output_Get:double)
   > Set FE Rotation(FE Rotation:Rotator="0, 0, 0") -> (Output_Get:Rotator)
   > Set Spinning(Spinning:bool=true) -> (Output_Get:bool)
   > Delay(Duration=5.5)
   > Set Spinning(Spinning=false)
   n5 = Event BlueprintUpdateAnimation()
   > Set Delta Time X(Delta Time X:double=@n5.DeltaTimeX) -> (Output_Get:double)
   > n6 = Sequence()
   > n7 = Cast To FortPlayerPawn(Object=GetOwningActor())
   > n8 = GetVariantChannelActiveIndexForItemActor!(
    execute
    Item:Athena Cosmetic Item Definition Object Reference=/BRCosmetics/Athena/Items/Cosmetics/Backpacks/Backpack_ForestJeanWink
    Actor:Actor Object Reference=@n7
    Channel:Gameplay Tag Structure=(TagName="Cosmetics.Variant.Channel.Parts")
    VariantList:Array of Gameplay Tag Structures=Make Array!("[0]"=(TagName="Cosmetics.Variant.Property.Stage2"), "[1]":Gameplay Tag Structure=(TagName="Cosmetics.Variant.Property.Stage1")) -> (Array)
    bErrorIfChannelMissing:bool=true
    bErrorIfTagNotInArray:bool=true
   ) -> (then, Index:int=0, bIndexIsValid:bool, bFoundChannel:bool)
   > Set Reactivity On/Off(Reactivity On/Off:bool="To Boolean (Integer)"!(InInt:int=@n8) -> (ReturnValue:bool)) -> (Output_Get:bool)
   > n9 = GetScalarParameterValue(Collection=Get MPC_Audio_Reactive_Parameters(), ParameterName=Get Amplitude() -> (Amplitude=None))
   > Set Amplitude_Value(Amplitude_Value=FInterpTo(Current=Get Amplitude_Value(), Target=@n9, DeltaTime=Get Delta Time X(), InterpSpeed=8))
   > n10 = Branch(Condition=Get bIsInFrontEnd())
   > n11 = Branch(Condition=Get bIsPlayingEmote())
   > Branch(Condition=Get bIsPlayingEmote())
   > n12 = Sequence()
   n13 = Get MusicActive() -> (MusicActive:bool)
   n14 = Get Amplitude_Value()
   n15 = Get ActivateThreshold() -> (ActivateThreshold:float)
   > n16 = Branch(Condition=AND Boolean(A=NOT Boolean(A=@n13), B="float > float"(A=@n14, B=@n15)))
   > Set MusicActive(MusicActive:bool=true) -> (Output_Get:bool)
   n17 = Get Amplitude_Value()
   Branch(execute=@n16.else, Condition=AND Boolean(A=Get MusicActive(), B="float < float"(A=@n17)))
   > Set MusicActive(MusicActive=false)
   n18 = Get MusicActive()
   Set ReactiveToMusic(execute=@n12.then_1, ReactiveToMusic=SelectFloat(A="To Float (Boolean)"(InBool=@n18), B=0, bPickA=Get Reactivity On/Off()))
   > n19 = Branch(Condition="float > float"(A=Get ReactiveToMusic()))
   > Set Time(Time=float + float(A=Get Time() -> (Time:double), B=Get Delta Time X()))
   n20 = Get Time()
   > Set RotZ_Value(RotZ_Value:float="% (Float)"!(A:double=float * float(A=@n20, B=135) {is=title K2Node_PromotableOperator}, B:double=360) -> (ReturnValue:double)) -> (Output_Get:float)
   > Set Rotation Speed(Rotation Speed:Rotator=MakeRotator(Yaw=Get RotZ_Value() -> (RotZ_Value:float))) -> (Output_Get:Rotator)
   Set Time(execute=@n19.else, Time=0)
   > n21 = Set RotZ_Value(RotZ_Value=FInterpTo_Constant!(Current:double=Get RotZ_Value(), Target:double=360, DeltaTime:double=Get Delta Time X(), InterpSpeed:double=135) -> (ReturnValue:double))
   > Set Rotation Speed(Rotation Speed=MakeRotator(Yaw=@n21))
   n22 = Sequence!(execute=@n11.else) -> (then_0, then_1, then_2:Exec)
   > Set Time(Time=float + float(A=Get Time(), B=Get Delta Time X()))
   Set ReactiveToMusic(execute=@n22.then_2, ReactiveToMusic=0)
   Branch(execute=@n22.then_1, Condition=Get Spinning() -> (Spinning:bool))
   n23 = Get Time()
   > Set FE Rotation(FE Rotation=MakeRotator(Yaw=float * float(A=@n23, B=125)))
   > Set FE Rotation(FE Rotation=RInterpTo!(
     Current:Rotator=Get FE Rotation()
     Target_Roll:float=0
     Target_Pitch:float=0
     Target_Yaw:float=360
     DeltaTime:float=Get Delta Time X()
     InterpSpeed:float=0.001
    ) -> (ReturnValue:Rotator))
   n24 = Sequence!(execute=@n10.else) -> (then_0, then_1)
   n25 = Get MusicActive()
   n26 = Get Amplitude_Value()
   n27 = Get ActivateThreshold()
   > n28 = Branch(Condition=AND Boolean(A=NOT Boolean(A=@n25), B="float > float"(A=@n26, B=@n27)))
   > Set MusicActive(MusicActive=true)
   n29 = Get DeactivateThreshold() -> (DeactivateThreshold:float)
   n30 = Get Amplitude_Value()
   Branch(execute=@n28.else, Condition=AND Boolean(A=Get MusicActive(), B="float < float"(A=@n30, B=@n29)))
   > Set MusicActive(MusicActive=false)
   n31 = Get MusicActive()
   Set ReactiveToMusic(execute=@n24.then_1, ReactiveToMusic=SelectFloat(A="To Float (Boolean)"(InBool=@n31), B=0, bPickA=Get Reactivity On/Off()))
   > n32 = Branch(Condition="float > float"(A=Get ReactiveToMusic()))
   > Set Time(Time=float + float(A=Get Time(), B=Get Delta Time X()))
   n33 = Get Time()
   > Set RotZ_Value(RotZ_Value="% (Float)"(A=float * float(A=@n33, B=135), B=360))
   > Set Rotation Speed(Rotation Speed=MakeRotator(Yaw=Get RotZ_Value()))
   Set Time(execute=@n32.else, Time=0)
   > n34 = Set RotZ_Value(RotZ_Value=FInterpTo_Constant(Current=Get RotZ_Value(), Target=360, DeltaTime=Get Delta Time X(), InterpSpeed=135))
   > Set Rotation Speed(Rotation Speed=MakeRotator(Yaw=@n34))
   n35 = Branch(execute=@n6.then_1, Condition=Get bIsSkyDiving())
   > Set GravityOverride(GravityOverride="0,0,980")
   Set GravityOverride(execute=@n35.else, GravityOverride="0,0,-980")
   Event BlueprintBeginPlay() {is=title K2Node_Event}
   > Parent: BlueprintBeginPlay() {is=title K2Node_CallParentFunction}
   > Delay(Duration=5.5)
   > Reactivity() {is=title K2Node_CallFunction}
   > Set Timer by Event(Delegate=@n3, Time=10, bLooping=true)
   Branch(Condition="float > float"(A=Get ReactiveToMusic()))
   n36 = Get Time()
   > Set Rotation Speed(Rotation Speed=MakeRotator(Yaw=float * float(A=@n36, B=135)))
   TryGetPawnOwner()
   n37 = "float > float"(A=Get!() -> (Rotation Speed_Roll:float, Rotation Speed_Pitch:float, Rotation Speed_Yaw:float).Rotation Speed_Yaw, B=180)
   Set Rotation Speed(Rotation Speed=RInterpTo(
     Current=Get Rotation Speed()
     Target_Roll=0
     Target_Pitch=0
     Target_Yaw=SelectFloat(A=360, B=0, bPickA=@n37)
     DeltaTime=Get Delta Time X()
     InterpSpeed=1
    ))
   Set Time(Time=float + float(A=Get Time(), B=Get Delta Time X()))
  ]}
]
variables [
 ActivateThreshold {type=number, default=0.6}
 Amplitud Float {type=number}
 Amplitude {type=string, default=Amplitude_Music}
 Amplitude_Value {type=number}
 Angular Spring Constant {type=number}
 Angular Spring Constant High {type=number}
 Angular Spring Constant Low {type=number}
 DeactivateThreshold {type=number, default=0.4}
 Delta Time X {type=number}
 FE Rotation {type=Rotator}
 Gravity Override Params_0 {
  type=FortGravityOverrideParameterSet
  default=(Skydive_GoingUp=(GravityOverride=()),Skydive_FloatingDown=(GravityOverride=()),Skydive_DivingDown=(GravityOverride=()),Skydive_Parachuting=(GravityOverride=()),OnGround_Standing=(GravityOverride=()),OnGround_Moving=(GravityOverride=()),OnGround_Moving_HighRotationRate=(GravityOverride=()),OnGround_Crouch_Moving=(GravityOverride=()),EmoteOrMelee=(GravityOverride=()),Swimming=(GravityOverride=()),Front_End=(GravityOverride=()),DBNO=(GravityOverride=()),DBNO_Carried=(GravityOverride=()),OnGround_Falling=(GravityOverride=()),Ziplining=(GravityOverride=()),Skydive_Surfing=(GravityOverride=()),Riding_Standing=(GravityOverride=()),Riding_Moving=(GravityOverride=()),Riding_Moving_Targeting=(GravityOverride=()),Tactical_Sprint=(GravityOverride=()),Rail_Grinding=(GravityOverride=()),)
 }
 GravityOverride {type=Vector}
 MPC_Audio_Reactive_Parameters {type=MaterialParameterCollection, default=/Game/Athena/Events/AudioAnalysis/MPC_Audio_Reactive_Parameters}
 MusicActive {type=boolean}
 ReactiveToMusic {type=number}
 Reactivity On/Off {type=boolean}
 Rotation Speed {type=Rotator}
 RotZ_Value {type=number}
 Spinning {type=boolean}
 Time {type=number}
 Time TEST {type=number}
]
